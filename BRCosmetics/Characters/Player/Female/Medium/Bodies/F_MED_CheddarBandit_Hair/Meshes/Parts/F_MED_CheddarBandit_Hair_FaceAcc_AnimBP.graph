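# Animation Blueprint for the CheddarBandit female hair/face-accessory part, adding secondary motion to hair bones.
# It copies the attached parent mesh's pose, additively rotates a hair base bone, then chains spring-driven, angle-limited AnimDynamics nodes through ten hair bones.
# The event graph maps head-to-neck yaw to HeadLimitAlpha, zeroed during emotes.
# That alpha feeds only a head and neck corrective rotation chain that is disconnected from the output pose.
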AnimBlueprint /BRCosmetics/Characters/Player/Female/Medium/Bodies/F_MED_CheddarBandit_Hair/Meshes/Parts/F_MED_CheddarBandit_Hair_FaceAcc_AnimBP
target_skeleton=&/F_MED_CheddarBandit_Hair_FaceAcc_Skeleton
graphs [
 AnimGraph {nodes [
   n1 = Get HeadLimitAlpha() -> (HeadLimitAlpha:double)
   n2 = "Transform (Modify) Bone - Bone: head"(Alpha=@n1, Translation="0,0,0", Rotation="0,0,35", Scale="1,1,1") {
    rotationMode=BMM_Additive
    scaleMode=BMM_Additive
    translationMode=BMM_Additive
    alphaBoolBlend {blendOption=Linear}
    boneToModify {boneName=head}
    rotation {roll=35}
   }
   n3 = "Transform (Modify) Bone - Bone: neck_02"(ComponentPose=@n2, Alpha=@n1, Translation="0,0,0", Rotation="0,0,10", Scale="1,1,1") {
    rotationMode=BMM_Additive
    scaleMode=BMM_Additive
    translationMode=BMM_Additive
    alphaBoolBlend {blendOption=Linear}
    boneToModify {boneName=neck_02}
    rotation {roll=10}
   }
   "Transform (Modify) Bone - Bone: neck_01"(ComponentPose=@n3, Alpha=@n1, Translation="0,0,0", Rotation="0,0,5", Scale="1,1,1") {
    rotationMode=BMM_Additive
    scaleMode=BMM_Additive
    translationMode=BMM_Additive
    alphaBoolBlend {blendOption=Linear}
    boneToModify {boneName=neck_01}
    rotation {roll=5}
   }
   n4 = "Transform (Modify) Bone - Bone: dyn_hair_hair_base_e"(
    ComponentPose=Local To Component(LocalPose=Copy Pose From Mesh() {bUseAttachedParent=true})
    Alpha=1
    Translation="0,0,0"
    Rotation="0,0,40"
    Scale="1,1,1"
   ) {
    rotationMode=BMM_Additive
    translationMode=BMM_Additive
    alphaBoolBlend {blendOption=Linear}
    boneToModify {boneName=dyn_hair_hair_base_e}
    rotation {roll=40}
   }
   n5 = Anim Dynamics - Bone: dyn_hair_hair_01_a(ComponentPose=@n4, Alpha=1, ExternalForce="0,0,0") {
    angularBiasOverride=0.5
    angularDampingOverride=0.999
    angularSpringConstant=1500
    bAngularSpring=true
    bOverrideAngularDamping=true
    bUseGravityOverride=true
    simulationSpace=BoneRelative
    alphaBoolBlend {blendOption=Linear}
    boundBone {boneName=dyn_hair_hair_01_a}
    componentAppliedLinearAccClamp=<3000 3000 3000>
    componentLinearAccScale=<0.7 0.7 0.7>
    componentLinearVelScale=<0.7 0.7 0.7>
    gravityOverride {z=-980}
    physicsBodyDefinitions [{
      boundBone {boneName=dyn_hair_hair_01_a}
      boxExtents=<20 20 20>
      constraintSetup {angularLimitsMax=<1 10 10>, angularLimitsMin=<-1 -10 -10>, angularTarget=<1 0 0>}
      localJointOffset=<0.1 0 0>
     }]
    relativeSpaceBone {boneName=root}
    simSpaceSettings {maxAngularAcceleration=5000, maxAngularVelocity=5000, simSpaceAngularAlpha=1}
   }
   n6 = Anim Dynamics - Bone: dyn_hair_hair_01_b(ComponentPose=@n5, Alpha=1, ExternalForce="0,0,0") {
    angularBiasOverride=0.5
    angularDampingOverride=0.999
    angularSpringConstant=1500
    bAngularSpring=true
    bOverrideAngularDamping=true
    bUseGravityOverride=true
    simulationSpace=BoneRelative
    alphaBoolBlend {blendOption=Linear}
    boundBone {boneName=dyn_hair_hair_01_b}
    componentAppliedLinearAccClamp=<3000 3000 3000>
    componentLinearAccScale=<0.7 0.7 0.7>
    componentLinearVelScale=<0.7 0.7 0.7>
    gravityOverride {z=-980}
    physicsBodyDefinitions [{
      boundBone {boneName=dyn_hair_hair_01_b}
      boxExtents=<20 20 20>
      constraintSetup {angularLimitsMax=<1 10 10>, angularLimitsMin=<-1 -10 -10>, angularTarget=<1 0 0>}
      localJointOffset=<0.1 0 0>
     }]
    relativeSpaceBone {boneName=root}
    simSpaceSettings {maxAngularAcceleration=5000, maxAngularVelocity=5000, simSpaceAngularAlpha=1}
   }
   n7 = Anim Dynamics - Bone: dyn_hair_hair_01_c(ComponentPose=@n6, Alpha=1, ExternalForce="0,0,0") {
    angularBiasOverride=0.5
    angularDampingOverride=0.999
    angularSpringConstant=1500
    bAngularSpring=true
    bOverrideAngularDamping=true
    bUseGravityOverride=true
    simulationSpace=BoneRelative
    alphaBoolBlend {blendOption=Linear}
    boundBone {boneName=dyn_hair_hair_01_c}
    componentAppliedLinearAccClamp=<3000 3000 3000>
    componentLinearAccScale=<0.7 0.7 0.7>
    componentLinearVelScale=<0.7 0.7 0.7>
    gravityOverride {z=-980}
    physicsBodyDefinitions [{
      boundBone {boneName=dyn_hair_hair_01_c}
      boxExtents=<20 20 20>
      constraintSetup {angularLimitsMax=<1 4 5>, angularLimitsMin=<-1 -10 -10>, angularTarget=<1 0 0>}
      localJointOffset=<0.1 0 0>
     }]
    relativeSpaceBone {boneName=root}
    simSpaceSettings {maxAngularAcceleration=5000, maxAngularVelocity=5000, simSpaceAngularAlpha=1}
   }
   n8 = Anim Dynamics - Bone: dyn_hair_hair_01_d(ComponentPose=@n7, Alpha=1, ExternalForce="0,0,0") {
    angularBiasOverride=0.5
    angularDampingOverride=0.999
    angularSpringConstant=1500
    bAngularSpring=true
    bOverrideAngularDamping=true
    bUseGravityOverride=true
    simulationSpace=BoneRelative
    alphaBoolBlend {blendOption=Linear}
    boundBone {boneName=dyn_hair_hair_01_d}
    componentAppliedLinearAccClamp=<3000 3000 3000>
    componentLinearAccScale=<0.7 0.7 0.7>
    componentLinearVelScale=<0.7 0.7 0.7>
    gravityOverride {z=-980}
    physicsBodyDefinitions [{
      boundBone {boneName=dyn_hair_hair_01_d}
      boxExtents=<20 20 20>
      constraintSetup {angularLimitsMax=<1 10 10>, angularLimitsMin=<-1 -10 -10>, angularTarget=<1 0 0>}
      localJointOffset=<0.1 0 0>
     }]
    relativeSpaceBone {boneName=root}
    simSpaceSettings {maxAngularAcceleration=5000, maxAngularVelocity=5000, simSpaceAngularAlpha=1}
   }
   n9 = Anim Dynamics - Bone: dyn_hair_hair_01_e(ComponentPose=@n8, Alpha=1, ExternalForce="0,0,0") {
    angularBiasOverride=0.5
    angularDampingOverride=0.999
    angularSpringConstant=1500
    bAngularSpring=true
    bOverrideAngularDamping=true
    bUseGravityOverride=true
    simulationSpace=BoneRelative
    alphaBoolBlend {blendOption=Linear}
    boundBone {boneName=dyn_hair_hair_01_e}
    componentAppliedLinearAccClamp=<3000 3000 3000>
    componentLinearAccScale=<0.7 0.7 0.7>
    componentLinearVelScale=<0.7 0.7 0.7>
    gravityOverride {z=-980}
    physicsBodyDefinitions [{
      boundBone {boneName=dyn_hair_hair_01_e}
      boxExtents=<20 20 20>
      constraintSetup {angularLimitsMax=<1 2 5>, angularLimitsMin=<-1 -25 -5>, angularTarget=<1 0 0>}
      localJointOffset=<0.1 0 0>
     }]
    relativeSpaceBone {boneName=root}
    simSpaceSettings {maxAngularAcceleration=5000, maxAngularVelocity=5000, simSpaceAngularAlpha=1}
   }
   n10 = Anim Dynamics - Bone: dyn_hair_hair_01_f(ComponentPose=@n9, Alpha=1, ExternalForce="0,0,0") {
    angularBiasOverride=0.5
    angularDampingOverride=0.999
    angularSpringConstant=1500
    bAngularSpring=true
    bOverrideAngularDamping=true
    bUseGravityOverride=true
    simulationSpace=BoneRelative
    alphaBoolBlend {blendOption=Linear}
    boundBone {boneName=dyn_hair_hair_01_f}
    componentAppliedLinearAccClamp=<3000 3000 3000>
    componentLinearAccScale=<0.7 0.7 0.7>
    componentLinearVelScale=<0.7 0.7 0.7>
    gravityOverride {z=-980}
    physicsBodyDefinitions [{
      boundBone {boneName=dyn_hair_hair_01_f}
      boxExtents=<20 20 20>
      constraintSetup {angularLimitsMax=<1 10 10>, angularLimitsMin=<-1 -10 -10>, angularTarget=<1 0 0>}
      localJointOffset=<0.1 0 0>
     }]
    relativeSpaceBone {boneName=root}
    simSpaceSettings {maxAngularAcceleration=5000, maxAngularVelocity=5000, simSpaceAngularAlpha=1}
   }
   n11 = Anim Dynamics - Bone: dyn_hair_hair_01_g(ComponentPose=@n10, Alpha=1, ExternalForce="0,0,0") {
    angularBiasOverride=0.5
    angularDampingOverride=0.999
    angularSpringConstant=1500
    bAngularSpring=true
    bOverrideAngularDamping=true
    bUseGravityOverride=true
    simulationSpace=BoneRelative
    alphaBoolBlend {blendOption=Linear}
    boundBone {boneName=dyn_hair_hair_01_g}
    componentAppliedLinearAccClamp=<3000 3000 3000>
    componentLinearAccScale=<0.7 0.7 0.7>
    componentLinearVelScale=<0.7 0.7 0.7>
    gravityOverride {z=-980}
    physicsBodyDefinitions [{
      boundBone {boneName=dyn_hair_hair_01_g}
      boxExtents=<20 20 20>
      constraintSetup {angularLimitsMax=<1 4 10>, angularLimitsMin=<-1 -10 -5>, angularTarget=<1 0 0>}
      localJointOffset=<0.1 0 0>
     }]
    relativeSpaceBone {boneName=root}
    simSpaceSettings {maxAngularAcceleration=5000, maxAngularVelocity=5000, simSpaceAngularAlpha=1}
   }
   n12 = Anim Dynamics - Bone: dyn_hair_hair_01_h(ComponentPose=@n11, Alpha=1, ExternalForce="0,0,0") {
    angularBiasOverride=0.5
    angularDampingOverride=0.999
    angularSpringConstant=2500
    bAngularSpring=true
    bOverrideAngularDamping=true
    bUseGravityOverride=true
    simulationSpace=BoneRelative
    alphaBoolBlend {blendOption=Linear}
    boundBone {boneName=dyn_hair_hair_01_h}
    componentAppliedLinearAccClamp=<4000 4000 4000>
    componentLinearAccScale=<1 1 1>
    componentLinearVelScale=<1 1 1>
    gravityOverride {z=-980}
    physicsBodyDefinitions [{
      boundBone {boneName=dyn_hair_hair_01_h}
      boxExtents=<20 20 20>
      constraintSetup {angularLimitsMax=<1 5 5>, angularLimitsMin=<-1 -5 -5>, angularTarget=<1 0 0>}
      localJointOffset=<0.05 0 0>
     }]
    relativeSpaceBone {boneName=head}
    simSpaceSettings {maxAngularAcceleration=5000, maxAngularVelocity=5000, simSpaceAngularAlpha=1}
   }
   n13 = Anim Dynamics - Bone: dyn_hair_hair_01_i(ComponentPose=@n12, Alpha=1, ExternalForce="0,0,0") {
    angularBiasOverride=0.5
    angularDampingOverride=0.999
    angularSpringConstant=1500
    bAngularSpring=true
    bOverrideAngularDamping=true
    bUseGravityOverride=true
    simulationSpace=BoneRelative
    alphaBoolBlend {blendOption=Linear}
    boundBone {boneName=dyn_hair_hair_01_i}
    componentAppliedLinearAccClamp=<3000 3000 3000>
    componentLinearAccScale=<0.7 0.7 0.7>
    componentLinearVelScale=<0.7 0.7 0.7>
    gravityOverride {z=-980}
    physicsBodyDefinitions [{
      boundBone {boneName=dyn_hair_hair_01_i}
      boxExtents=<20 20 20>
      constraintSetup {angularLimitsMax=<1 10 10>, angularLimitsMin=<-1 -10 -10>, angularTarget=<1 0 0>}
      localJointOffset=<0.1 0 0>
     }]
    relativeSpaceBone {boneName=root}
    simSpaceSettings {maxAngularAcceleration=5000, maxAngularVelocity=5000, simSpaceAngularAlpha=1}
   }
   n14 = Anim Dynamics - Bone: dyn_hair_hair_01_j(ComponentPose=@n13, Alpha=1, ExternalForce="0,0,0") {
    angularBiasOverride=0.5
    angularDampingOverride=0.999
    angularSpringConstant=1500
    bAngularSpring=true
    bOverrideAngularDamping=true
    bUseGravityOverride=true
    simulationSpace=BoneRelative
    alphaBoolBlend {blendOption=Linear}
    boundBone {boneName=dyn_hair_hair_01_j}
    componentAppliedLinearAccClamp=<3000 3000 3000>
    componentLinearAccScale=<0.7 0.7 0.7>
    componentLinearVelScale=<0.7 0.7 0.7>
    gravityOverride {z=-980}
    physicsBodyDefinitions [{
      boundBone {boneName=dyn_hair_hair_01_j}
      boxExtents=<20 20 20>
      constraintSetup {angularLimitsMax=<1 10 10>, angularLimitsMin=<-1 -10 -10>, angularTarget=<1 0 0>}
      localJointOffset=<0.1 0 0>
     }]
    relativeSpaceBone {boneName=root}
    simSpaceSettings {maxAngularAcceleration=5000, maxAngularVelocity=5000, simSpaceAngularAlpha=1}
   }
   Output Pose(Result=Component To Local(ComponentPose=@n14)) {layerGroup=None}
  ]}
 EventGraph {nodes [
   Event BlueprintUpdateAnimation()
   > n1 = Branch(Condition=Get bIsPlayingEmote())
   > Set HeadLimitAlpha(HeadLimitAlpha:double=0) -> (Output_Get:double)
   n2 = BreakTransform(InTransform=GetDeltaTransformFromRefPose(self=GetOwnerSkeletalMeshComponent(), BoneName=head, BaseName=neck_02))
   Set HeadLimitAlpha(execute=@n1.else, HeadLimitAlpha=MapRangeClamped(Value=BreakRotator(InRot=@n2.Rotation).Yaw, InRangeA=-10, InRangeB=-35))
  ]}
]
variables [HeadLimitAlpha {type=number}]
